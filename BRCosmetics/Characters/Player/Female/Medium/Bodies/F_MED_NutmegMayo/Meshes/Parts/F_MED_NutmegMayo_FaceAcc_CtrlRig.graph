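# Control Rig for the F_MED_NutmegMayo face accessory that applies corrective offsets to the three back hair master bones as the character poses.
# Spherical pose readers on the head and both upper arms weight chained Modify Transforms nodes that rotate, shift or squash those bones.
# Two arm correctives are also scaled by pistol or rifle and not-emote alphas.
# A final zero-offset head chain and the Head_Rot_Bk_Alpha setter sit unconnected to execution.
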ControlRigBlueprint /BRCosmetics/Characters/Player/Female/Medium/Bodies/F_MED_NutmegMayo/Meshes/Parts/F_MED_NutmegMayo_FaceAcc_CtrlRig
graphs [
 Multiply_1_1_ContainedGraph {nodes [
   n1 = RigVMFunctionEntryNode()
   RigVMFunctionReturnNode(Result=&RigVMFunction_MathDoubleMul(B=@n1.C, A=&RigVMFunction_MathDoubleMul(A=@n1.A, B=@n1.B).Result).Result)
  ]}
 Multiply_3_ContainedGraph {nodes [
   n1 = RigVMFunctionEntryNode()
   Multiply = &RigVMFunction_MathDoubleMul(A=@n1.A, B=@n1.B)
   RigVMFunctionReturnNode(Result=&RigVMFunction_MathDoubleMul(B=@n1.C, A=@Multiply.Result).Result)
  ]}
 RigVMFunctionLibrary {links [], nodes []}
 RigVMModel {nodes [
   RigUnit_BeginExecution = &RigUnit_BeginExecution()
   n1 = &RigUnit_SphericalPoseReader(
    ActiveRegionScaleFactors=(PositiveWidth=1,NegativeWidth=1,PositiveHeight=1,NegativeHeight=1)
    ActiveRegionSize=0.1
    Debug=(bDrawDebug=false,DebugScale=25,DebugSegments=20,DebugThickness=0.25)
    DriverAxis=(X=0,Y=-1,Z=0)
    DriverItem=(Type=Bone,Name="head")
    FalloffRegionScaleFactors=(PositiveWidth=2,NegativeWidth=2,PositiveHeight=2,NegativeHeight=2)
    FalloffSize=0.2
    FlipWidthScaling=false
    OptionalParentItem=(Type=Bone,Name="neck_01")
    RotationOffset=(X=-90,Y=0,Z=-35)
    ExecutePin=@RigUnit_BeginExecution.ExecutePin
   )
   > &RigUnit_ModifyTransforms(
    ItemToModify=((Item=(Type=Bone,Name="Hair_bk_c_MSTR"),Transform=(Rotation=(X=0,Y=0,Z=0,W=1),Translation=(X=0,Y=0,Z=0),Scale3D=(X=0.8,Y=1,Z=1))))
    ItemToModify.0.Transform.Rotation=&RigVMFunction_MathQuaternionFromEuler(Euler=(X=-15,Y=0,Z=0)).Result
    Weight=@n1.OutputParam
   )
   > &RigUnit_ModifyTransforms(
    ItemToModify=((Item=(Type=Bone,Name="Hair_bk_l_MSTR"),Transform=(Rotation=(X=0,Y=0,Z=0,W=1),Translation=(X=0,Y=0,Z=0),Scale3D=(X=0.8,Y=1,Z=1))))
    ItemToModify.0.Transform.Rotation=&RigVMFunction_MathQuaternionFromEuler(Euler=(X=-10,Y=0,Z=0)).Result
    Weight=@n1.OutputParam
   )
   > &RigUnit_ModifyTransforms(
    ItemToModify=((Item=(Type=Bone,Name="Hair_bk_r_MSTR"),Transform=(Rotation=(X=0,Y=0,Z=0,W=1),Translation=(X=0,Y=0,Z=0),Scale3D=(X=0.8,Y=1,Z=1))))
    ItemToModify.0.Transform.Rotation=&RigVMFunction_MathQuaternionFromEuler(Euler=(X=-10,Y=0,Z=0)).Result
    Weight=@n1.OutputParam
   )
   > n2 = &RigUnit_SphericalPoseReader(
    ActiveRegionScaleFactors=(PositiveWidth=0.5,NegativeWidth=0.5,PositiveHeight=0.5,NegativeHeight=0.5)
    ActiveRegionSize=0.1
    Debug=(bDrawDebug=false,bDraw2D=false,bDrawLocalAxes=false,DebugScale=25,DebugSegments=20,DebugThickness=0.25)
    DriverAxis=(X=-1,Y=0,Z=0)
    DriverItem=(Type=Bone,Name="upperarm_r")
    FalloffRegionScaleFactors=(PositiveWidth=0.5,NegativeWidth=0.5,PositiveHeight=0.5,NegativeHeight=0.5)
    FalloffSize=0.25
    OptionalParentItem=(Type=Bone,Name="spine_05")
    RotationOffset=(X=-100,Y=45,Z=30)
   )
   Multiply = RigVMAggregateNode(A=@n2.OutputParam, B=RigVMVariableNode() {variable=Is Pistol Alpha}.Value, C=RigVMVariableNode() {variable=Not Playing Emote Alpha}.Value)
   > &RigUnit_ModifyTransforms(
    ItemToModify=((Item=(Type=Bone,Name="Hair_bk_c_MSTR"),Transform=(Rotation=(X=0,Y=0,Z=0,W=1),Translation=(X=0,Y=-1,Z=0),Scale3D=(X=1,Y=1,Z=1))))
    ItemToModify.0.Transform.Rotation=&RigVMFunction_MathQuaternionFromEuler(Euler=(X=-5,Y=0,Z=30)).Result
    Weight=@Multiply.Result
   )
   > &RigUnit_ModifyTransforms(
    ItemToModify=((Item=(Type=Bone,Name="Hair_bk_l_MSTR"),Transform=(Rotation=(X=0,Y=0,Z=0,W=1),Translation=(X=0,Y=0,Z=0),Scale3D=(X=1,Y=1,Z=1))))
    ItemToModify.0.Transform.Rotation=&RigVMFunction_MathQuaternionFromEuler(Euler=(X=0,Y=0,Z=0)).Result
    Weight=@Multiply.Result
   )
   > &RigUnit_ModifyTransforms(
    ItemToModify=((Item=(Type=Bone,Name="Hair_bk_r_MSTR"),Transform=(Rotation=(X=0,Y=0,Z=0,W=1),Translation=(X=-3,Y=-2,Z=0),Scale3D=(X=1,Y=1,Z=1))))
    ItemToModify.0.Transform.Rotation=&RigVMFunction_MathQuaternionFromEuler(Euler=(X=15,Y=0,Z=-20)).Result
    Weight=@Multiply.Result
   )
   > n3 = &RigUnit_SphericalPoseReader(
    ActiveRegionScaleFactors=(PositiveWidth=0.5,NegativeWidth=0.5,PositiveHeight=0.5,NegativeHeight=0.5)
    ActiveRegionSize=0.1
    Debug=(bDrawDebug=false,bDraw2D=false,bDrawLocalAxes=false,DebugScale=25,DebugSegments=20,DebugThickness=0.25)
    DriverAxis=(X=1,Y=0,Z=0)
    DriverItem=(Type=Bone,Name="upperarm_l")
    FalloffRegionScaleFactors=(PositiveWidth=1.75,NegativeWidth=1.75,PositiveHeight=1.75,NegativeHeight=1.75)
    FalloffSize=0.2
    OptionalParentItem=(Type=Bone,Name="spine_05")
    RotationOffset=(X=15,Y=0,Z=0)
   )
   Multiply_1 = RigVMAggregateNode(A=@n3.OutputParam, B=RigVMVariableNode() {variable=Is Rifle Alpha}.Value, C=RigVMVariableNode() {variable=Not Playing Emote Alpha}.Value)
   > &RigUnit_ModifyTransforms(
    ItemToModify=((Item=(Type=Bone,Name="Hair_bk_c_MSTR"),Transform=(Rotation=(X=0,Y=0,Z=0,W=1),Translation=(X=0,Y=0,Z=0),Scale3D=(X=1,Y=1,Z=1))))
    ItemToModify.0.Transform.Rotation=&RigVMFunction_MathQuaternionFromEuler(Euler=(X=15,Y=0,Z=0)).Result
    Weight=@Multiply_1.Result
   )
   > &RigUnit_ModifyTransforms(
    ItemToModify=((Item=(Type=Bone,Name="Hair_bk_l_MSTR"),Transform=(Rotation=(X=0,Y=0,Z=0,W=1),Translation=(X=0,Y=0,Z=0),Scale3D=(X=1,Y=1,Z=1))))
    ItemToModify.0.Transform.Rotation=&RigVMFunction_MathQuaternionFromEuler(Euler=(X=5,Y=0,Z=0)).Result
    Weight=@Multiply_1.Result
   )
   > &RigUnit_ModifyTransforms(
    ItemToModify=((Item=(Type=Bone,Name="Hair_bk_r_MSTR"),Transform=(Rotation=(X=0,Y=0,Z=0,W=1),Translation=(X=0,Y=-2,Z=0),Scale3D=(X=1,Y=1,Z=1))))
    ItemToModify.0.Transform.Rotation=&RigVMFunction_MathQuaternionFromEuler(Euler=(X=20,Y=0,Z=20)).Result
    Weight=@Multiply_1.Result
   )
   > n4 = &RigUnit_SphericalPoseReader(
    ActiveRegionScaleFactors=(PositiveWidth=1,NegativeWidth=1,PositiveHeight=1,NegativeHeight=1)
    ActiveRegionSize=0.1
    Debug=(bDrawDebug=false,bDraw2D=false,bDrawLocalAxes=false,DebugScale=25,DebugSegments=20,DebugThickness=0.25)
    DriverAxis=(X=-1,Y=0,Z=0)
    DriverItem=(Type=Bone,Name="upperarm_r")
    FalloffRegionScaleFactors=(PositiveWidth=1,NegativeWidth=1,PositiveHeight=1,NegativeHeight=1)
    FalloffSize=0.35
    OptionalParentItem=(Type=Bone,Name="spine_05")
    RotationOffset=(X=190,Y=-10,Z=0)
   )
   > &RigUnit_ModifyTransforms(
    ItemToModify=((Item=(Type=Bone,Name="Hair_bk_r_MSTR"),Transform=(Rotation=(X=0,Y=0,Z=0,W=1),Translation=(X=0,Y=0,Z=0),Scale3D=(X=1,Y=1,Z=1))))
    ItemToModify.0.Transform.Rotation=&RigVMFunction_MathQuaternionFromEuler(Euler=(X=10,Y=0,Z=0)).Result
    Weight=@n4.OutputParam
   )
   > &RigUnit_ModifyTransforms(
    ItemToModify=((Item=(Type=Bone,Name="Hair_bk_c_MSTR"),Transform=(Rotation=(X=0,Y=0,Z=0,W=1),Translation=(X=0,Y=0,Z=0),Scale3D=(X=1,Y=1,Z=1))))
    ItemToModify.0.Transform.Rotation=&RigVMFunction_MathQuaternionFromEuler(Euler=(X=-5,Y=0,Z=0)).Result
    Weight=@n4.OutputParam
   )
   > n5 = &RigUnit_SphericalPoseReader(
    ActiveRegionScaleFactors=(PositiveWidth=1,NegativeWidth=1,PositiveHeight=1,NegativeHeight=1)
    ActiveRegionSize=0.1
    Debug=(bDrawDebug=false,bDraw2D=false,bDrawLocalAxes=false,DebugScale=25,DebugSegments=20,DebugThickness=0.25)
    DriverAxis=(X=1,Y=0,Z=0)
    DriverItem=(Type=Bone,Name="upperarm_l")
    FalloffRegionScaleFactors=(PositiveWidth=1,NegativeWidth=1,PositiveHeight=1,NegativeHeight=1)
    FalloffSize=0.35
    OptionalParentItem=(Type=Bone,Name="spine_05")
    RotationOffset=(X=10,Y=-10,Z=0)
   )
   > &RigUnit_ModifyTransforms(
    ItemToModify=((Item=(Type=Bone,Name="Hair_bk_l_MSTR"),Transform=(Rotation=(X=0,Y=0,Z=0,W=1),Translation=(X=0,Y=0,Z=0),Scale3D=(X=1,Y=1,Z=1))))
    ItemToModify.0.Transform.Rotation=&RigVMFunction_MathQuaternionFromEuler(Euler=(X=-5,Y=0,Z=0)).Result
    Weight=@n5.OutputParam
   )
   > &RigUnit_ModifyTransforms(
    ItemToModify=((Item=(Type=Bone,Name="Hair_bk_c_MSTR"),Transform=(Rotation=(X=0,Y=0,Z=0,W=1),Translation=(X=0,Y=0,Z=0),Scale3D=(X=1,Y=1,Z=1))))
    ItemToModify.0.Transform.Rotation=&RigVMFunction_MathQuaternionFromEuler(Euler=(X=-5,Y=0,Z=0)).Result
    Weight=@n5.OutputParam
   )
   > n6 = &RigUnit_SphericalPoseReader(
    ActiveRegionScaleFactors=(PositiveWidth=0.5,NegativeWidth=0.5,PositiveHeight=1,NegativeHeight=1)
    ActiveRegionSize=0.1
    Debug=(bDrawDebug=false,DebugScale=25,DebugSegments=20,DebugThickness=0.25)
    DriverAxis=(X=0,Y=-1,Z=0)
    DriverItem=(Type=Bone,Name="head")
    FalloffRegionScaleFactors=(PositiveWidth=0.5,NegativeWidth=0.5,PositiveHeight=1,NegativeHeight=1)
    FalloffSize=0.2
    FlipWidthScaling=false
    OptionalParentItem=(Type=Bone,Name="spine_05")
    RotationOffset=(X=-90,Y=0,Z=-90)
   )
   > &RigUnit_ModifyTransforms(
    ItemToModify=((Item=(Type=Bone,Name="Hair_bk_c_MSTR"),Transform=(Rotation=(X=0,Y=0,Z=0,W=1),Translation=(X=0,Y=-2,Z=2),Scale3D=(X=0.5,Y=1,Z=1))))
    ItemToModify.0.Transform.Rotation=&RigVMFunction_MathQuaternionFromEuler(Euler=(X=-15,Y=0,Z=0)).Result
    Weight=@n6.OutputParam
   )
   > &RigUnit_ModifyTransforms(
    ItemToModify=((Item=(Type=Bone,Name="Hair_bk_l_MSTR"),Transform=(Rotation=(X=0,Y=0,Z=0,W=1),Translation=(X=0,Y=-2,Z=2),Scale3D=(X=0.5,Y=1,Z=1))))
    ItemToModify.0.Transform.Rotation=&RigVMFunction_MathQuaternionFromEuler(Euler=(X=-15,Y=0,Z=0)).Result
    Weight=@n6.OutputParam
   )
   > &RigUnit_ModifyTransforms(
    ItemToModify=((Item=(Type=Bone,Name="Hair_bk_r_MSTR"),Transform=(Rotation=(X=0,Y=0,Z=0,W=1),Translation=(X=0,Y=-2,Z=2),Scale3D=(X=0.5,Y=1,Z=1))))
    ItemToModify.0.Transform.Rotation=&RigVMFunction_MathQuaternionFromEuler(Euler=(X=-15,Y=0,Z=0)).Result
    Weight=@n6.OutputParam
   )
   n7 = &RigUnit_SphericalPoseReader(
    ActiveRegionScaleFactors=(PositiveWidth=1,NegativeWidth=0.5,PositiveHeight=1,NegativeHeight=1)
    ActiveRegionSize=0.1
    Debug=(bDrawDebug=false,DebugScale=25,DebugSegments=20,DebugThickness=0.25)
    DriverAxis=(X=0,Y=-1,Z=0)
    DriverItem=(Type=Bone,Name="head")
    ExecutePin=()
    FalloffRegionScaleFactors=(PositiveWidth=1,NegativeWidth=0.5,PositiveHeight=1,NegativeHeight=1)
    FalloffSize=0.25
    OptionalParentItem=(Type=Bone,Name="neck_02")
    RotationOffset=(X=-55,Y=0,Z=-10)
   )
   n8 = &RigVMFunction_MathDoubleMul(A=@n7.OutputParam, B=RigVMVariableNode() {variable=Head_Rot_Bk_Alpha}.Value)
   > &RigUnit_ModifyTransforms(
    ItemToModify=((Item=(Type=Bone,Name="Hair_bk_r_MSTR"),Transform=(Rotation=(X=0,Y=0,Z=0,W=1),Translation=(X=0,Y=0,Z=0),Scale3D=(X=1,Y=1,Z=1))))
    ItemToModify.0.Transform.Rotation=&RigVMFunction_MathQuaternionFromEuler(Euler=(X=0,Y=0,Z=0)).Result
    Weight=@n8.Result
   )
   > &RigUnit_ModifyTransforms(
    ItemToModify=((Item=(Type=Bone,Name="Hair_bk_c_MSTR"),Transform=(Rotation=(X=0,Y=0,Z=0,W=1),Translation=(X=0,Y=0,Z=0),Scale3D=(X=1,Y=1,Z=1))))
    ItemToModify.0.Transform.Rotation=&RigVMFunction_MathQuaternionFromEuler(Euler=(X=0,Y=0,Z=0)).Result
    Weight=@n8.Result
   )
   > &RigUnit_ModifyTransforms(
    ItemToModify=((Item=(Type=Bone,Name="Hair_bk_l_MSTR"),Transform=(Rotation=(X=0,Y=0,Z=0,W=1),Translation=(X=0,Y=0,Z=0),Scale3D=(X=1,Y=1,Z=1))))
    ItemToModify.0.Transform.Rotation=&RigVMFunction_MathQuaternionFromEuler(Euler=(X=0,Y=0,Z=0)).Result
    Weight=@n8.Result
   )
   RigVMVariableNode(ExecuteContext=(), Value=@n1.OutputParam) {variable=Head_Rot_Bk_Alpha}
  ]}
]
hierarchy [
 root {loc=<0 0 0>, rot=<0 0 0 1>}
 pelvis {parent=root, loc=<0 -0.5301 98.0571>, rot=<0.0127 0.707 -0.0127 -0.707>}
 spine_01 {parent=pelvis, loc=<6.1208 -0.9328 0>, rot=<0 0 0.034 -0.9994>}
 spine_02 {parent=spine_01, loc=<6.5996 -0.2095 0>, rot=<0 0 0.0238 -0.9997>}
 spine_03 {parent=spine_02, loc=<6.4914 -0.1237 0>, rot=<0 0 0.0149 -0.9999>}
 spine_04 {parent=spine_03, loc=<6.9598 0.0794 0>, rot=<0 0 -0.0132 -0.9999>}
 spine_05 {parent=spine_04, loc=<6.9987 0 0>, rot=<0 0 -0.0015 -1>}
 neck_01 {parent=spine_05, loc=<11.3138 0.1529 0>, rot=<0 0 -0.2003 -0.9797>}
 neck_02 {parent=neck_01, loc=<4.5627 0 0>, rot=<0 0 0.0504 -0.9987>}
 head {parent=neck_02, loc=<5.3914 0 0>, rot=<0 0 0.115 -0.9934>}
 FACIAL_C_FacialRoot {parent=head, loc=<2.1524 -0.746 -0.0002>, rot=<0.4879 -0.5118 -0.4879 -0.5118>}
 FACIAL_C_Forehead {parent=FACIAL_C_FacialRoot, loc=<0.0006 -5.3014 9.8104>, rot=<-0.1079 -0.0001 -0.0001 -0.9942>}
 FACIAL_L_ForeheadIn {parent=FACIAL_C_FacialRoot, loc=<2.3538 -5.2635 9.7737>, rot=<-0.1249 -0.0698 -0.0004 -0.9897>}
 FACIAL_R_ForeheadIn {parent=FACIAL_C_FacialRoot, loc=<-2.3566 -5.2611 9.7696>, rot=<-0.1249 0.0696 0.0001 -0.9897>}
 FACIAL_L_ForeheadMid {parent=FACIAL_C_FacialRoot, loc=<4.4179 -5.4357 9.1074>, rot=<-0.1343 -0.174 -0.032 -0.975>}
 FACIAL_R_ForeheadMid {parent=FACIAL_C_FacialRoot, loc=<-4.4217 -5.4297 9.1006>, rot=<-0.1343 0.1739 0.0318 -0.9751>}
 FACIAL_L_ForeheadOut {parent=FACIAL_C_FacialRoot, loc=<6.4466 -5.0616 7.9408>, rot=<-0.1829 -0.25 -0.0956 -0.946>}
 FACIAL_R_ForeheadOut {parent=FACIAL_C_FacialRoot, loc=<-6.4488 -5.0579 7.937>, rot=<-0.1829 0.2499 0.0954 -0.9461>}
 FACIAL_L_EyesackUpper {parent=FACIAL_C_FacialRoot, loc=<4.2817 -4.1947 9.6219>, rot=<-0.0155 -0.08 -0.07 -0.9942>}
 FACIAL_R_EyesackUpper {parent=FACIAL_C_FacialRoot, loc=<-4.2908 -4.1896 9.6235>, rot=<-0.0155 0.0798 0.0697 -0.9943>}
 FACIAL_L_EyelidUpperFurrow {parent=FACIAL_C_FacialRoot, loc=<4.181 -3.8798 9.5551>, rot=<-0.0936 -0.1454 -0.0894 -0.9809>}
 FACIAL_R_EyelidUpperFurrow {parent=FACIAL_C_FacialRoot, loc=<-4.1878 -3.8793 9.5586>, rot=<-0.0937 0.1453 0.0891 -0.9809>}
 FACIAL_L_EyelidUpperA {parent=FACIAL_C_FacialRoot, loc=<3.4955 -2.036 5.861>, rot=<-0.2114 -0.0001 -0.0001 -0.9774>}
 FACIAL_R_EyelidUpperA {parent=FACIAL_C_FacialRoot, loc=<-3.4757 -2.0521 5.866>, rot=<-0.2114 -0.0001 -0.0001 -0.9774>}
 FACIAL_L_EyelidLowerA {parent=FACIAL_C_FacialRoot, loc=<3.4955 -2.036 5.861>, rot=<0.2214 0 -0.0002 -0.9752>}
 FACIAL_R_EyelidLowerA {parent=FACIAL_C_FacialRoot, loc=<-3.4757 -2.052 5.866>, rot=<0.2214 0 -0.0002 -0.9752>}
 FACIAL_L_EyeCornerInner {parent=FACIAL_C_FacialRoot, loc=<2.3536 -1.452 8.8921>, rot=<-0.0646 -0.0001 -0.0001 -0.9979>}
 FACIAL_R_EyeCornerInner {parent=FACIAL_C_FacialRoot, loc=<-2.3291 -1.4503 8.8804>, rot=<-0.0646 -0.0001 -0.0001 -0.9979>}
 FACIAL_L_EyeCornerOuter {parent=FACIAL_C_FacialRoot, loc=<6.3497 -2.0444 7.7461>, rot=<-0.0597 -0.3819 -0.0248 -0.922>}
 FACIAL_R_EyeCornerOuter {parent=FACIAL_C_FacialRoot, loc=<-6.373 -1.9747 7.7315>, rot=<-0.0598 0.3817 0.0246 -0.922>}
 FACIAL_L_EyesackLower {parent=FACIAL_C_FacialRoot, loc=<4.5616 0.6507 7.6718>, rot=<-0.0862 -0.1381 -0.0314 -0.9862>}
 FACIAL_R_EyesackLower {parent=FACIAL_C_FacialRoot, loc=<-4.5605 0.6787 7.6686>, rot=<-0.0862 0.1379 0.0311 -0.9862>}
 FACIAL_L_CheekInner {parent=FACIAL_C_FacialRoot, loc=<2.6647 1.298 8.8355>, rot=<-0.1856 -0.0752 -0.2025 -0.9586>}
 FACIAL_R_CheekInner {parent=FACIAL_C_FacialRoot, loc=<-2.662 1.2982 8.8396>, rot=<-0.1856 0.075 0.2023 -0.9587>}
 FACIAL_L_CheekOuter {parent=FACIAL_C_FacialRoot, loc=<5.4626 1.6642 6.9044>, rot=<-0.0043 -0.3827 0.0757 -0.9208>}
 FACIAL_R_CheekOuter {parent=FACIAL_C_FacialRoot, loc=<-5.46 1.6608 6.902>, rot=<-0.0044 0.3826 -0.0759 -0.9208>}
 FACIAL_L_NasolabialBulge {parent=FACIAL_C_FacialRoot, loc=<4.5122 5.3627 7.4974>, rot=<-0.0019 -0.3782 -0.0008 -0.9257>}
 FACIAL_R_NasolabialBulge {parent=FACIAL_C_FacialRoot, loc=<-4.501 5.3751 7.5027>, rot=<-0.002 0.378 0.0005 -0.9258>}
 FACIAL_L_CheekLower {parent=FACIAL_C_FacialRoot, loc=<4.4897 5.4511 6.7134>, rot=<-0.0237 -0.4298 0.0299 -0.9021>}
 FACIAL_R_CheekLower {parent=FACIAL_C_FacialRoot, loc=<-4.4786 5.4646 6.7189>, rot=<-0.0238 0.4296 -0.0302 -0.9022>}
 FACIAL_C_Nose {parent=FACIAL_C_FacialRoot, loc=<0.0018 0.4033 10.091>, rot=<-0.1942 -0.0001 -0.0001 -0.981>}
 FACIAL_L_Nostril {parent=FACIAL_C_Nose, loc=<1.0712 3.0355 -0.0073>, rot=<0.0173 0.1305 0.0023 -0.9913>}
 FACIAL_R_Nostril {parent=FACIAL_C_Nose, loc=<-1.0733 3.034 -0.0136>, rot=<0.0173 -0.1305 -0.0023 -0.9913>}
 FACIAL_C_NoseTip {parent=FACIAL_C_Nose, loc=<-0.0038 3.0262 0.756>, rot=<0.0872 0 0 -0.9962>}
 FACIAL_L_Masseter {parent=FACIAL_C_FacialRoot, loc=<9.1399 2.9639 4.4421>, rot=<0.0052 0.1121 0.015 -0.9936>}
 FACIAL_R_Masseter {parent=FACIAL_C_FacialRoot, loc=<-9.1484 2.9682 4.4392>, rot=<0.0052 -0.1122 -0.0153 -0.9936>}
 FACIAL_C_NoseBridge {parent=FACIAL_C_FacialRoot, loc=<0 -2.9648 10.7783>, rot=<-0.0208 -0.0001 -0.0001 -0.9998>}
 FACIAL_L_EyelidUpperB {parent=FACIAL_C_FacialRoot, loc=<3.4955 -2.036 5.861>, rot=<-0.2875 -0.0001 -0.0001 -0.9578>}
 FACIAL_L_EyelidLowerB {parent=FACIAL_C_FacialRoot, loc=<3.4955 -2.036 5.861>, rot=<0.2553 0 -0.0002 -0.9669>}
 FACIAL_R_EyelidLowerB {parent=FACIAL_C_FacialRoot, loc=<-3.4757 -2.0521 5.866>, rot=<0.2638 0 -0.0002 -0.9646>}
 FACIAL_R_EyelidUpperB {parent=FACIAL_C_FacialRoot, loc=<-3.4757 -2.0521 5.866>, rot=<-0.2875 -0.0001 -0.0001 -0.9578>}
 Hair_bk_c_MSTR {parent=FACIAL_C_FacialRoot, loc=<-0.0003 -4.9604 -10.1357>, rot=<-0.687 0 0 -0.7267>}
 Hair_bk_c_01 {parent=Hair_bk_c_MSTR, loc=<0 0 0>, rot=<-0.5 -0.5 0.5 -0.5>}
 Hair_bk_c_02 {parent=Hair_bk_c_01, loc=<13.693 0 0>, rot=<0 0 0 1>}
 Hair_bk_l_MSTR {parent=FACIAL_C_FacialRoot, loc=<8.2112 -4.809 -7.4361>, rot=<-0.687 0 0 -0.7267>}
 Hair_bk_l_01 {parent=Hair_bk_l_MSTR, loc=<0 0 0>, rot=<-0.5 -0.5 0.5 -0.5>}
 Hair_bk_l_02 {parent=Hair_bk_l_01, loc=<13.693 0 0>, rot=<0 0 0 1>}
 Hair_bk_r_MSTR {parent=FACIAL_C_FacialRoot, loc=<-8.211 -4.809 -7.4361>, rot=<-0.687 0 0 -0.7267>}
 Hair_bk_r_01 {parent=Hair_bk_r_MSTR, loc=<0 0 0>, rot=<-0.5 -0.5 0.5 -0.5>}
 Hair_bk_r_02 {parent=Hair_bk_r_01, loc=<13.693 0 0>, rot=<0 0 0 1>}
 clavicle_l {parent=spine_05, loc=<5.4831 -0.1053 -3.2946>, rot=<0.0465 -0.7289 -0.009 -0.6829>}
 upperarm_l {parent=clavicle_l, loc=<9.8658 -1.2136 -0.4105>, rot=<0.0164 -0.3245 -0.0269 -0.9454>}
 clavicle_r {parent=spine_05, loc=<5.4831 -0.1053 3.2946>, rot=<0.7289 0.0465 -0.6829 0.009>}
 upperarm_r {parent=clavicle_r, loc=<-9.8658 1.2136 0.4105>, rot=<0.0164 -0.3245 -0.0269 -0.9454>}
]
variables [
 Head_Rot_Bk_Alpha {type=float, default=0}
 Is Pistol Alpha {type=double, default=0}
 Is Rifle Alpha {type=double, default=0}
 Not Playing Emote Alpha {type=double, default=0}
]
